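; Skyfire knockback launcher rocket projectile that explodes on hit or lifetime timeout, launching enemies and its own shooter.
; On authority, each hit adds a stacking GameplayEffect whose count scales knockback, with an impact multiplier near the blast center.
; Launches get at least 45-degree lift, except close self-hits, which push opposite a pitch-offset aim; enemies also receive a movement-blocking effect.
; Remotes handle audio and VFX cleanup; stopping triggers one-shot impact Niagara.
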(event EventTick
  (params (DeltaSeconds :type Float%20%28single-precision%29)))

(event EventBeginPlay
  (bind _tracermesh (Variables|Default|GetTracerMesh))
  ;[0.03086 1 0.5887] Set max lifetime timer to trigger explosion.
  (|CallParentFunction)
  (bind _returnvalue (Utilities|Time|SetTimerbyEvent (delegate Custom|MaxLifetime) (Variables|Default|GetPrjLifetime)))
  (Variables|Default|SetLifetimeTImer _returnvalue)
  (bind _asfort_player_pawn (Utilities|Casting|CastToFortPlayerPawn (Variables|Actor|GetInstigator)))
  (Utilities|IsValid (Variables|/Script/FortniteGame.FortPawn|GetCurrentWeapon _asfort_player_pawn)
    (:Is%20Valid
      (Rendering|Material|SetVectorParameterValueonMaterials _tracermesh (Variables|Default|GetParameterNameBulletOrigin) (Effects|GetMuzzleLocation (Variables|/Script/FortniteGame.FortPawn|GetCurrentWeapon _asfort_player_pawn) -1))
      (Rendering|Material|SetScalarParameterValueonMaterials _tracermesh (Variables|Default|GetParameterNameStartTime) (Utilities|Time|GetGameTimeinSeconds))
      (Rendering|Material|SetScalarParameterValueonMaterials _tracermesh (Variables|Default|GetParameterNameWPODuration) (Variables|Default|GetParameterValueWPODuration)))))

(event Custom|MaxLifetime
  ;[0.03086 1 0.5887] Set max lifetime timer to trigger explosion.
  (Delivery|DoExplosion))

(event EventOnExploded
  (params (HitActors :type Array%20of%20Actor%20Object%20References) (HitResults :type Array%20of%20Hit%20Result%20Structures))
  (bind _returnvalue (Game|GetInstigator))
  (bind _ns_trail (Variables|Default|GetNS_Trail))
  ;[1 0.5788 0] Exploded. Clear lifetime timer and handle knockback/visuals.
  (|CallParentFunction HitActors HitResults)
  (Utilities|Time|ClearandInvalidateTimerbyHandle (Variables|Default|GetLifetimeTImer))
  (Utilities|FlowControl|Sequence
    (:then_0
      (Networking|SwitchHasAuthority
        (:Authority
          ;;[0.2255 0.2209 1] Apply stacking knockback GEs and launch the target based on the distance to the explosion center.
          (Variables|Default|SetHitLocation (Transformation|GetActorLocation))
          (for _array_element HitActors
            (bind _asfort_pawn (Utilities|Casting|CastToFortPawn _array_element))
            (if (CallFunction|ShouldHitTarget _asfort_pawn)
              (Utilities|IsValid (Ability|GetAbilitySystemComponent _asfort_pawn)
                (:Is%20Valid
                  (GameplayEffects|ApplyGameplayEffectToSelf :self (Ability|GetAbilitySystemComponent _asfort_pawn) :GameplayEffectClass "/Skyfire/Gameplay/KnockbackWeapon/GE_Skyfire_StackingKnockbackCount.GE_Skyfire_StackingKnockbackCount_C" :Level 0.0)
                  (if (not (Team|OnSameTeam _returnvalue _asfort_pawn))
                    (GameplayEffects|ApplyGameplayEffectToSelf :self (Ability|GetAbilitySystemComponent _asfort_pawn) :GameplayEffectClass "/Skyfire/Gameplay/KnockbackWeapon/GE_Skyfire_PreventMoveWhenHit.GE_Skyfire_PreventMoveWhenHit_C" :Level 0.0))
                  (bind (_returnvalue_x _returnvalue_y _returnvalue_z) (Math|Vector|BreakVector (* (CallFunction|DetermineKnockbackValues _asfort_pawn (GameplayEffects|GetGameplayEffectCount (Ability|GetAbilitySystemComponent _asfort_pawn) "/Skyfire/Gameplay/KnockbackWeapon/GE_Skyfire_StackingKnockbackCount.GE_Skyfire_StackingKnockbackCount_C" 0 true)) (.KnockbackMagnitude (CallFunction|DetermineKnockbackValues _asfort_pawn (GameplayEffects|GetGameplayEffectCount (Ability|GetAbilitySystemComponent _asfort_pawn) "/Skyfire/Gameplay/KnockbackWeapon/GE_Skyfire_StackingKnockbackCount.GE_Skyfire_StackingKnockbackCount_C" 0 true))))))
                  (Character|LaunchCharacter _asfort_pawn (Math|Vector|SelectVector (* (* (Math|Vector|GetForwardVector (Math|Rotator|MakeRotator (.roll (Pawn|GetBaseAimRotation _asfort_pawn)) (+ (.pitch (Pawn|GetBaseAimRotation _asfort_pawn)) (Variables|Default|GetYOffset)) (.yaw (Pawn|GetBaseAimRotation _asfort_pawn)))) -1.0) (ScalableFloat|GetValueatLevel (Variables|Default|GetSF_InitialKnockbackStrength))) (Math|Vector|MakeVector _returnvalue_x _returnvalue_y (Math|Float|Max(Float) (Math|Float|Absolute(Float) _returnvalue_z) (Math|Vector|VectorLengthXY (Math|Vector|MakeVector _returnvalue_x _returnvalue_y _returnvalue_z)))) (and (.IsDirectHit (CallFunction|DetermineKnockbackValues _asfort_pawn (GameplayEffects|GetGameplayEffectCount (Ability|GetAbilitySystemComponent _asfort_pawn) "/Skyfire/Gameplay/KnockbackWeapon/GE_Skyfire_StackingKnockbackCount.GE_Skyfire_StackingKnockbackCount_C" 0 true))) (Utilities|Equal(Object) (Game|GetInstigator) _asfort_pawn))) (not (Team|OnSameTeam _returnvalue _asfort_pawn)) (not (Team|OnSameTeam _returnvalue _asfort_pawn))))))))
        (:Remote
          ;;[1 0.0245 0.998] Destroy projectile VFX and play explosion FX.
          (Components|DestroyComponent (Variables|Default|GetNS_Projectile))
          (Audio|PlaySoundatLocation "/Skyfire/Gameplay/KnockbackWeapon/Audio/Knockback_Launcher_Prj_Explo_Cue.Knockback_Launcher_Prj_Explo_Cue" (Transformation|GetActorLocation))
          (Transformation|DetachFromComponent _ns_trail "KeepWorld" "KeepWorld" "KeepWorld" true)
          (Components|Activation|Deactivate _ns_trail)
          (Audio|Components|Audio|FadeOut (Variables|Default|GetRocket_AudioComponent) 0.2 0.0 "Linear" :!class /Script/Engine.AudioComponent)
          (CallFunction|PlayProjectileHitSFX HitActors))) ; Server Only, since all our relevant hits will come from the server
    )))

(event Custom|Play%20Projectile%20Hit%20SFX
  (params (InputPin :type Array%20of%20Actor%20Object%20References))
  ;[1 0.1177 0.6181] Play Projectile Hit SFX on all players hit by projectile
  (Networking|CanExecuteCosmeticEvents
    (:True
      (for _array_element InputPin
        (bind _asfort_pawn (Utilities|Casting|CastToFortPawn _array_element))
        (if (CallFunction|ShouldHitTarget _asfort_pawn)
          (if (and (CallFunction|IsLocal (Game|GetInstigator)) (not (CallFunction|IsLocal _asfort_pawn)))
            (Audio|PlaySound2D (Variables|Sound|GetSFX_Impact2D))
            (else
              _))
          (Audio|SpawnSoundAttached (select (CallFunction|IsLocal _asfort_pawn) (Variables|Sound|GetSFX_KnockBack_P1) (Variables|Sound|GetSFX_KnockBackP3)) (Variables|/Script/Engine.Character|GetMesh _asfort_pawn))
          (if (not (Utilities|Equal(Object) (Game|GetInstigator) _asfort_pawn))
            (Audio|SpawnSoundAttached (select (CallFunction|IsLocal _asfort_pawn) (Variables|Sound|GetSFX_KnockBack_Spark_P1) (Variables|Sound|GetSFX_KnockBack_Spark_P3)) (Variables|/Script/Engine.Character|GetMesh _asfort_pawn))))))))

(event EventOnStop
  (params (Hit :type Hit%20Result%20Structure%20%28by%20ref%29))
  ;[0 0.4385 1] Call multicasted event to play FX, handles client not always triggering stop
  (|CallParentFunction Hit)
  (CallFunction|PlayExplosionFX Hit))

(event Custom|PlayExplosionFX
  (params (HitResult :type Hit%20Result%20Structure))
  (bind (_bblockinghit _binitialoverlap _time _distance _location _impactpoint _normal) (Collision|BreakHitResult HitResult))
  ;[0.6562 1 0] Spawn the explosion FX based on the impact.
  (Utilities|FlowControl|DoOnce false
    (:Completed
      (Niagara|SpawnSystematLocation "/Skyfire/Gameplay/KnockbackWeapon/FX/Impact/NS_KnockbackWeapon_Impact.NS_KnockbackWeapon_Impact" _location (Math|Rotator|MakeRotfromZ _normal)))))

(fn DetermineKnockbackValues
  (params (Target :type Actor%20Object%20Reference) (EffectStackCount :type Integer))
  ;[0 0.1029 1] Calculate the knockback direction and strength based on if it were a direct hit or explosion and return the values.
  (Utilities|IsValid Target
    (:Is%20Valid
      (return (Math|Vector|Normalize (- (Transformation|GetActorLocation Target) (Variables|Default|GetHitLocation)) 0.01) (* (Variables|GetEffectStackCount) (* (Math|Float|SelectFloat (ScalableFloat|GetValueatLevel (Variables|Default|GetImpactMult)) 1.0 (> (Math|Float|Square (ScalableFloat|GetValueatLevel (Variables|Default|GetExplosionStartDistance))) (Math|Vector|DistanceSquared(Vector) (Transformation|GetActorLocation Target) (Variables|Default|GetHitLocation)))) (ScalableFloat|GetValueatLevel (Variables|Default|GetSF_InitialKnockbackStrength)))) (> (Math|Float|Square (ScalableFloat|GetValueatLevel (Variables|Default|GetExplosionStartDistance))) (Math|Vector|DistanceSquared(Vector) (Transformation|GetActorLocation Target) (Variables|Default|GetHitLocation)))))
    (:Is%20Not%20Valid
      ;;[1 0 0.007519] No valid target, return.
      (return "0, 0, 0" 0.0 false))))

(fn IsLocal
  (params (Pawn :type Object%20Reference))
  (Networking|CanExecuteCosmeticEvents
    (:True
      (for _array_element (Fortnite|GetLocalFortPlayerControllers true)
        (Utilities|IsValid (Pawn|GetControlledPawn _array_element)
          (:Is%20Valid
            (if (Utilities|Equal(Object) (Pawn|GetControlledPawn _array_element) Pawn)
              (return true)))) ; Check against valid as we don't want \nNull=Null to succeed.
      ))
    (:False
      _))
  (return false))

(fn ShouldHitTarget
  (params (Target :type Fort%20Pawn%20Object%20Reference))
  (bind _returnvalue (Game|GetInstigator))
  ;[0.04774 0.5205 1] Check if we should hit the target
  (Utilities|IsValid _returnvalue
    (:Is%20Valid
      (return (or (not (Team|OnSameTeam Target _returnvalue)) (Utilities|Equal(Object) Target _returnvalue))))))

(fn ConstructionScript
  (Development|PrintString ""))
